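# Material function that supplies clear coat amount, roughness and top normal from three parameterized textures.
# Each texture has its own offset, scale and rotation, with a custom SinCos node feeding MF_TransformUVs for its UVs.
# Amount takes the red channel and roughness the green, each multiplied by a scalar factor.
# The normal reuses its SinCos in MF_RotateNormals_TS to rotate tangent-space normals by the UV rotation angle.
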MaterialFunction /Emporium/MaterialFunctions/MF_Clearcoat
graph {nodes [
  n1 = VectorParameter() {
   group=ClearCoat
   parameterName=ClearCoatTexture_OffsetScale
   sortPriority=5
   channelNames {a=Scale V, b=Scale U, g=Offset V, r=Offset U}
   defaultValue=<rgba 0 0 1 1>
  }
  MaterialExpressionScalarParameter_1 = ScalarParameter() {desc=Rad, group=ClearCoat, parameterName=ClearCoatTexture_Rotation, sortPriority=6}
  n2 = Custom!(Angle=@MaterialExpressionScalarParameter_1) {
   code="float s=0.0;\r\nfloat c=1.0;\r\nsincos(Angle, s, c);\r\nreturn float2(s,c);"
   description=SinCos
   outputType=CMOT_Float2
   inputs [{inputName=Angle, input {expression=&self:MaterialExpressionScalarParameter_1}}]
  }
  n3 = TextureSampleParameter2D(UVs=MaterialFunctionCall!(OffsetAndScale=AppendVector(A=@n1, B=@n1), SinCosAngle=@n2) {materialFunction=&/MF_TransformUVs}.Result) {
   desc=linear
   group=ClearCoat
   parameterName=ClearCoatTexture
   samplerSource=SSM_Wrap_WorldGroupSettings
   samplerType=SAMPLERTYPE_Masks
   sortPriority=0
   texture=/Game/Global/Textures/Default/Blanks/T_White
   channelNames {a="", b="", g="", r=ClearCoat}
  }
  FunctionOutput(None=Multiply(A=@n3.ClearCoat, B=ScalarParameter() {defaultValue=1, group=ClearCoat, parameterName=ClearCoatFactor, sliderMax=1, sortPriority=3})) {outputName=ClearCoatAmount}
  n4 = VectorParameter() {
   group=ClearCoat
   parameterName=ClearCoatRoughnessTexture_OffsetScale
   sortPriority=7
   channelNames {a=Scale V, b=Scale U, g=Offset V, r=Offset U}
   defaultValue=<rgba 0 0 1 1>
  }
  MaterialExpressionScalarParameter_13 = ScalarParameter() {desc=Rad, group=ClearCoat, parameterName=ClearCoatRoughnessTexture_Rotation, sortPriority=8}
  n5 = Custom(Angle=@MaterialExpressionScalarParameter_13) {
   code="float s=0.0;\r\nfloat c=1.0;\r\nsincos(Angle, s, c);\r\nreturn float2(s,c);"
   description=SinCos
   outputType=CMOT_Float2
   inputs [{inputName=Angle, input {expression=&self:MaterialExpressionScalarParameter_13}}]
  }
  n6 = TextureSampleParameter2D(UVs=MaterialFunctionCall(OffsetAndScale=AppendVector(A=@n4, B=@n4), SinCosAngle=@n5) {materialFunction=&MF_TransformUVs}.Result) {
   desc=linear
   group=ClearCoat
   parameterName=ClearCoatRoughnessTexture
   samplerSource=SSM_Wrap_WorldGroupSettings
   samplerType=SAMPLERTYPE_Masks
   sortPriority=1
   texture=&T_White
   channelNames {a="", b="", g=Roughness, r=""}
  }
  FunctionOutput(None=Multiply(A=@n6.Roughness, B=ScalarParameter() {defaultValue=1, group=ClearCoat, parameterName=ClearCoatRoughnessFactor, sliderMax=1, sortPriority=4})) {outputName=ClearCoatRoughness, sortPriority=1}
  n7 = VectorParameter() {
   group=ClearCoat
   parameterName=ClearCoatNormalTexture_OffsetScale
   sortPriority=9
   channelNames {a=Scale V, b=Scale U, g=Offset V, r=Offset U}
   defaultValue=<rgba 0 0 1 1>
  }
  MaterialExpressionScalarParameter_0 = ScalarParameter() {desc=Rad, group=ClearCoat, parameterName=ClearCoatNormalTexture_Rotation, sortPriority=10}
  n8 = Custom(Angle=@MaterialExpressionScalarParameter_0) {
   code="float s=0.0;\r\nfloat c=1.0;\r\nsincos(Angle, s, c);\r\nreturn float2(s,c);"
   description=SinCos
   outputType=CMOT_Float2
   inputs [{inputName=Angle, input {expression=&self:MaterialExpressionScalarParameter_0}}]
  }
  n9 = TextureSampleParameter2D(UVs=MaterialFunctionCall(OffsetAndScale=AppendVector(A=@n7, B=@n7), SinCosAngle=@n8) {materialFunction=&MF_TransformUVs}.Result) {
   desc=Top Normal
   group=ClearCoat
   parameterName=ClearCoatNormalTexture
   samplerSource=SSM_Wrap_WorldGroupSettings
   samplerType=SAMPLERTYPE_Normal
   sortPriority=2
   texture=&FlatNormalBlue
  }
  FunctionOutput(None=MaterialFunctionCall!(NormalsTS=@n9.RGB, SinCos=@n8) {materialFunction=&/MF_RotateNormals_TS}.Result) {outputName=ClearCoatTopNormal, sortPriority=2}
 ]}
