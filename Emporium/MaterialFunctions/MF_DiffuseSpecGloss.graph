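# Material function that supplies specular-glossiness workflow surface inputs: DiffuseColor, SpecularColor, Glossiness and Opacity.
# Diffuse and specular-glossiness textures, white by default, each get their own offset, scale and radian rotation through MF_TransformUVs, with custom HLSL sincos nodes.
# Factor parameters multiply the samples, with Glossiness taken from specular-glossiness alpha and Opacity from diffuse alpha.
# Both textures use the shared wrap sampler.
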MaterialFunction /Emporium/MaterialFunctions/MF_DiffuseSpecGloss
graph {nodes [
  n1 = VectorParameter() {
   group=DiffuseSpecGloss
   parameterName=DiffuseTexture_OffsetScale
   sortPriority=5
   channelNames {a=Scale V, b=Scale U, g=Offset V, r=Offset U}
   defaultValue=<rgba 0 0 1 1>
  }
  MaterialExpressionScalarParameter_8 = ScalarParameter() {desc=Rad, group=DiffuseSpecGloss, parameterName=DiffuseTexture_Rotation, sortPriority=6}
  n2 = Custom!(Angle=@MaterialExpressionScalarParameter_8) {
   code="float s=0.0;\r\nfloat c=1.0;\r\nsincos(Angle, s, c);\r\nreturn float2(s,c);"
   description=SinCos
   outputType=CMOT_Float2
   inputs [{inputName=Angle, input {expression=&self:MaterialExpressionScalarParameter_8}}]
  }
  n3 = TextureSampleParameter2D(UVs=MaterialFunctionCall!(OffsetAndScale=AppendVector(A=@n1, B=@n1), SinCosAngle=@n2) {materialFunction=&/MF_TransformUVs}.Result) {
   desc=srgb
   group=DiffuseSpecGloss
   parameterName=DiffuseTexture
   samplerSource=SSM_Wrap_WorldGroupSettings
   sortPriority=0
   texture=&T_WhitesRGB
  }
  n4 = VectorParameter() {group=DiffuseSpecGloss, parameterName=DiffuseFactor, sortPriority=2, defaultValue=<rgba 1 1 1 1>}
  FunctionOutput(None=Multiply(A=@n3.RGB, B=@n4.RGB)) {outputName=DiffuseColor}
  n5 = VectorParameter() {
   group=DiffuseSpecGloss
   parameterName=SpecularGlossinessTexture_OffsetScale
   sortPriority=7
   channelNames {a=Scale V, b=Scale U, g=Offset V, r=Offset U}
   defaultValue=<rgba 0 0 1 1>
  }
  MaterialExpressionScalarParameter_1 = ScalarParameter() {desc=Rad, group=DiffuseSpecGloss, parameterName=SpecularGlossinessTexture_Rotation, sortPriority=8}
  n6 = Custom(Angle=@MaterialExpressionScalarParameter_1) {
   code="float s=0.0;\r\nfloat c=1.0;\r\nsincos(Angle, s, c);\r\nreturn float2(s,c);"
   description=SinCos
   outputType=CMOT_Float2
   inputs [{inputName=Angle, input {expression=&self:MaterialExpressionScalarParameter_1}}]
  }
  n7 = TextureSampleParameter2D(UVs=MaterialFunctionCall(OffsetAndScale=AppendVector(A=@n5, B=@n5), SinCosAngle=@n6) {materialFunction=&MF_TransformUVs}.Result) {
   desc=srgb
   group=DiffuseSpecGloss
   parameterName=SpecularGlossinessTexture
   samplerSource=SSM_Wrap_WorldGroupSettings
   sortPriority=1
   texture=&T_WhitesRGB
  }
  FunctionOutput(None=Multiply(A=@n7.RGB, B=VectorParameter() {group=DiffuseSpecGloss, parameterName=SpecFactor, sortPriority=3, defaultValue=<rgba 1 1 1 0>}.RGB)) {outputName=SpecularColor, sortPriority=1}
  FunctionOutput(None=Multiply(A=@n7.A, B=ScalarParameter() {defaultValue=1, group=DiffuseSpecGloss, parameterName=GlossinessFactor, sliderMax=1, sortPriority=4})) {outputName=Glossiness, sortPriority=2}
  FunctionOutput(None=Multiply(A=@n3.A, B=@n4.A)) {outputName=Opacity, sortPriority=3}
 ]}
