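# Material function that outputs an emissive color as a tinted texture sample, exposing its parameters in the Emissive group.
# MF_TransformUVs builds the sampling UVs from an offset/scale vector parameter and a rotation in radians, whose sine and cosine come from a Custom HLSL node.
# The EmissiveTexture sample is multiplied by EmissiveFactor, which defaults to black, so emission is off until that factor is overridden.
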MaterialFunction /Emporium/MaterialFunctions/MF_Emissive
graph {nodes [
  n1 = VectorParameter() {
   group=Emissive
   parameterName=EmissiveTexture_OffsetScale
   sortPriority=2
   channelNames {a=Scale V, b=Scale U, g=Offset V, r=Offset U}
   defaultValue=<rgba 0 0 1 1>
  }
  MaterialExpressionScalarParameter_12 = ScalarParameter() {desc=Rad, group=Emissive, parameterName=EmissiveTexture_Rotation, sortPriority=3}
  n2 = Custom!(Angle=@MaterialExpressionScalarParameter_12) {
   code="float s=0.0;\r\nfloat c=1.0;\r\nsincos(Angle, s, c);\r\nreturn float2(s,c);"
   description=SinCos
   outputType=CMOT_Float2
   inputs [{inputName=Angle, input {expression=&self:MaterialExpressionScalarParameter_12}}]
  }
  n3 = TextureSampleParameter2D(UVs=MaterialFunctionCall!(OffsetAndScale=AppendVector(A=@n1, B=@n1), SinCosAngle=@n2) {materialFunction=&/MF_TransformUVs}.Result) {
   desc=srgb
   group=Emissive
   parameterName=EmissiveTexture
   samplerSource=SSM_Wrap_WorldGroupSettings
   sortPriority=0
   texture=&T_WhitesRGB
  }
  FunctionOutput(None=Multiply(A=@n3.RGB, B=VectorParameter() {group=Emissive, parameterName=EmissiveFactor, sortPriority=1, defaultValue=<rgba 0 0 0 1>}.RGB)) {outputName=Emissive}
 ]}
