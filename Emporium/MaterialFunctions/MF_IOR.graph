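# MF_IOR is a material function that converts a scalar index of refraction parameter into a Specular value.
# It clamps the "ior" parameter to the 1 to 2 range, then squares (1 - ior)/(1 + ior) to get F0 reflectance.
# F0 is divided by 0.08 and saturated to a 0-1 Specular output, so an IOR of 1.5 gives 0.5.
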MaterialFunction /Emporium/MaterialFunctions/MF_IOR
graph {nodes [
  n1 = Clamp(None=ScalarParameter() {defaultValue=2, group=Specular, parameterName=ior, sliderMax=2, sliderMin=1, sortPriority=2}) {maxDefault=2, minDefault=1}
  n2 = Divide(A=OneMinus(None=@n1), B=Add(A=@n1))
  n3 = Multiply(A=@n2, B=@n2) {desc=F0 Reflectance}
  FunctionOutput(None=Saturate(None=Divide(A=@n3) {constB=0.08, desc="Normalize F0 to 0-1 Specular value.\r\nIOR 1.5 = F0 0.04 = Spec 0.5"})) {outputName=Specular}
 ]}
