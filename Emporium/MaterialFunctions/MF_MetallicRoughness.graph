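# Material function that outputs Metallic and Roughness from one packed mask texture, with roughness in green and metallic in blue, matching the glTF channel layout.
# An offset/scale vector and a rotation in radians, converted by a custom SinCos node, feed MF_TransformUVs to produce the sampling UVs.
# Each sampled channel is multiplied by its own scalar factor, and all parameters share the MetallicRoughness group.
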MaterialFunction /Emporium/MaterialFunctions/MF_MetallicRoughness
graph {nodes [
  n1 = VectorParameter() {
   group=MetallicRoughness
   parameterName=MetallicRoughnessTexture_OffsetScale
   sortPriority=3
   channelNames {a=Scale V, b=Scale U, g=Offset V, r=Offset U}
   defaultValue=<rgba 0 0 1 1>
  }
  MaterialExpressionScalarParameter_9 = ScalarParameter() {desc=Rad, group=MetallicRoughness, parameterName=MetallicRoughnessTexture_Rotation, sortPriority=4}
  n2 = Custom!(Angle=@MaterialExpressionScalarParameter_9) {
   code="float s=0.0;\r\nfloat c=1.0;\r\nsincos(Angle, s, c);\r\nreturn float2(s,c);"
   description=SinCos
   outputType=CMOT_Float2
   inputs [{inputName=Angle, input {expression=&self:MaterialExpressionScalarParameter_9}}]
  }
  n3 = TextureSampleParameter2D(UVs=MaterialFunctionCall!(OffsetAndScale=AppendVector(A=@n1, B=@n1), SinCosAngle=@n2) {materialFunction=&/MF_TransformUVs}.Result) {
   desc=Linear
   group=MetallicRoughness
   parameterName=MetallicRoughnessTexture
   samplerSource=SSM_Wrap_WorldGroupSettings
   samplerType=SAMPLERTYPE_Masks
   sortPriority=0
   texture=/Game/Global/Textures/Default/Blanks/T_White
   channelNames {a="", b=Metallic, g=Roughness, r=""}
  }
  FunctionOutput(None=Multiply(A=ScalarParameter() {defaultValue=1, group=MetallicRoughness, parameterName=MetallicFactor, sliderMax=1, sortPriority=1}, B=@n3.Metallic)) {outputName=Metallic}
  FunctionOutput(None=Multiply(A=@n3.Roughness, B=ScalarParameter() {defaultValue=1, group=MetallicRoughness, parameterName=RoughnessFactor, sliderMax=1, sortPriority=2})) {outputName=Roughness, sortPriority=1}
 ]}
