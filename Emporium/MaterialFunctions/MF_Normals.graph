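# Material function that outputs a tangent-space normal from a parameterised normal map with adjustable offset, scale, rotation and strength.
# A Custom HLSL node computes the rotation's sine and cosine once, shared by MF_TransformUVs for the UVs and MF_RotateNormals_TS to rotate sampled normals to match.
# The texture uses the shared wrap sampler.
# Normal Scale is saturated and inverted to drive FlattenNormal before output.
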MaterialFunction /Emporium/MaterialFunctions/MF_Normals
graph {nodes [
  n1 = VectorParameter() {
   group=Normal
   parameterName=NormalTexture_OffsetScale
   sortPriority=2
   channelNames {a=Scale V, b=Scale U, g=Offset V, r=Offset U}
   defaultValue=<rgba 0 0 1 1>
  }
  MaterialExpressionScalarParameter_7 = ScalarParameter() {desc=Rad, group=Normal, parameterName=NormalTexture_Rotation, sortPriority=3}
  n2 = Custom!(Angle=@MaterialExpressionScalarParameter_7) {
   code="float s=0.0;\r\nfloat c=1.0;\r\nsincos(Angle, s, c);\r\nreturn float2(s,c);"
   description=SinCos
   outputType=CMOT_Float2
   inputs [{inputName=Angle, input {expression=&self:MaterialExpressionScalarParameter_7}}]
  }
  n3 = TextureSampleParameter2D(UVs=MaterialFunctionCall!(OffsetAndScale=AppendVector(A=@n1, B=@n1), SinCosAngle=@n2) {materialFunction=&/MF_TransformUVs}.Result) {
   desc=BottomNormal
   group=Normal
   parameterName=NormalTexture
   samplerSource=SSM_Wrap_WorldGroupSettings
   samplerType=SAMPLERTYPE_Normal
   sortPriority=0
   texture=&FlatNormalBlue
  }
  n4 = Saturate(None=ScalarParameter() {defaultValue=1, group=Normal, parameterName=Normal Scale, sliderMax=1, sortPriority=1})
  FunctionOutput(None=MaterialFunctionCall(
    Normal=MaterialFunctionCall!(NormalsTS=@n3.RGB, SinCos=@n2) {materialFunction=&/MF_RotateNormals_TS}.Result
    Flatness=OneMinus(None=@n4)
   ) {materialFunction=&FlattenNormal}.Result)
 ]}
