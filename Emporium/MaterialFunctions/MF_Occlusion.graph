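# Material function that supplies an ambient-style Occlusion output from a parameterized mask texture, grouped under Occlusion parameters.
# Offset/scale and rotation parameters, with sine and cosine from a Custom HLSL node, feed MF_TransformUVs to set the texture's UVs.
# The red channel of OcclusionTexture, white by default and using the shared world-group wrap sampler, is blended from 1 by saturated OcclusionStrength.
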MaterialFunction /Emporium/MaterialFunctions/MF_Occlusion
graph {nodes [
  n1 = VectorParameter() {
   group=Occlusion
   parameterName=OcclusionTexture_OffsetScale
   sortPriority=2
   channelNames {a=Scale V, b=Scale U, g=Offset V, r=Offset U}
   defaultValue=<rgba 0 0 1 1>
  }
  MaterialExpressionScalarParameter_11 = ScalarParameter() {desc=Rad, group=Occlusion, parameterName=OcclusionTexture_Rotation, sortPriority=3}
  n2 = Custom!(Angle=@MaterialExpressionScalarParameter_11) {
   code="float s=0.0;\r\nfloat c=1.0;\r\nsincos(Angle, s, c);\r\nreturn float2(s,c);"
   description=SinCos
   outputType=CMOT_Float2
   inputs [{inputName=Angle, input {expression=&self:MaterialExpressionScalarParameter_11}}]
  }
  n3 = TextureSampleParameter2D(UVs=MaterialFunctionCall!(OffsetAndScale=AppendVector(A=@n1, B=@n1), SinCosAngle=@n2) {materialFunction=&/MF_TransformUVs}.Result) {
   desc=linear
   group=Occlusion
   parameterName=OcclusionTexture
   samplerSource=SSM_Wrap_WorldGroupSettings
   samplerType=SAMPLERTYPE_Masks
   sortPriority=0
   texture=/Game/Global/Textures/Default/Blanks/T_White
  }
  n4 = ScalarParameter() {defaultValue=1, group=Occlusion, parameterName=OcclusionStrength, sliderMax=1, sortPriority=1}
  FunctionOutput(None=LinearInterpolate(A=Constant() {r=1}, B=@n3.R, Alpha=Saturate(None=@n4))) {outputName=Occlusion}
 ]}
