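# Translucent, surface-lit PBR base material for Emporium, assembled from shared material functions for base color, normals, emissive, metallic-roughness and opacity.
# A static switch selects specular-glossiness, converted to metal-roughness, or native metallic-roughness inputs.
# A clearcoat switch selects Clear Coat or Default Lit shading and the top-coat normal, while the base normal feeds the clearcoat normal output.
# IOR, specular and occlusion multiply into one value.
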Material /Emporium/BaseMaterials/M_Translucent
graph {
 properties {
  bScreenSpaceReflections=true
  bUsedWithInstancedStaticMeshes=true
  blendMode=BLEND_Translucent
  opacityMaskClipValue=0.5
  pixelDepthOffsetMode=PDOM_Legacy
  shadingModel=MSM_FromMaterialExpression
  translucencyLightingMode=TLM_SurfacePerPixelLighting
  usedShadingModels=Default Lit | Clear Coat
 }
 nodes [
  n1 = MaterialFunctionCall() {materialFunction=/Emporium/MaterialFunctions/MF_Normals}
  ClearCoatNormalCustomOutput!(None=@n1.Result)
  MaterialFunctionCall() {materialFunction=/Emporium/MaterialFunctions/MF_Emissive}
  Multiply(
   A=Multiply(
    A=MaterialFunctionCall() {materialFunction=/Emporium/MaterialFunctions/MF_IOR}.Specular
    B=MaterialFunctionCall() {materialFunction=/Emporium/MaterialFunctions/MF_Specular}.Specular
   )
   B=MaterialFunctionCall() {materialFunction=/Emporium/MaterialFunctions/MF_Occlusion}.Occlusion
  )
  StaticSwitchParameter(True=ShadingModel() {shadingModel=MSM_ClearCoat}, False=ShadingModel()) {group=ClearCoat, parameterName=Using Clearcoat, sortPriority=-1}
  n2 = MaterialFunctionCall() {materialFunction=/Emporium/MaterialFunctions/MF_DiffuseSpecGloss}
  n3 = MaterialFunctionCall!(DiffuseColor=@n2.DiffuseColor, SpecularColor=@n2.DiffuseColor, Glosiness=@n2.DiffuseColor) {materialFunction=/Emporium/MaterialFunctions/MF_SpecGlossToMetalRoughness}
  n4 = MaterialFunctionCall() {materialFunction=/Emporium/MaterialFunctions/MF_Basecolor}
  StaticSwitchParameter(True=@n3."BaseColor (V3)", False=@n4.BaseColor) {group=PBRWorkflow, parameterName=Using SpecularGlossiness, sortPriority=0}
  StaticSwitchParameter(True=MaterialFunctionCall() {materialFunction=/Emporium/MaterialFunctions/MF_Clearcoat}.ClearCoatTopNormal, False=@n1.Result) {group=ClearCoat, parameterName=Using Clearcoat, sortPriority=-1}
  n5 = MaterialFunctionCall() {materialFunction=/Emporium/MaterialFunctions/MF_MetallicRoughness}
  StaticSwitchParameter(True=@n3."Metallic (S)", False=@n5.Metallic) {group=PBRWorkflow, parameterName=Using SpecularGlossiness, sortPriority=0}
  StaticSwitchParameter(True=@n3.Roughness, False=@n5.Roughness) {group=PBRWorkflow, parameterName=Using SpecularGlossiness, sortPriority=0}
  StaticSwitchParameter(True=@n2.Opacity, False=@n4.Opacity) {group=PBRWorkflow, parameterName=Using SpecularGlossiness, sortPriority=0}
 ]
}
