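# MF_Fresnel_DS is a material function that outputs a double-sided Schlick Fresnel term.
# Specular is scaled by 0.08 into a base reflectance, which blends toward one by a power of the inverted view-normal dot product.
# The tangent-space normal input is transformed before the camera-vector dot, and an Abs keeps backfaces valid in ray tracing.
# The exponent is exposed for artistic control and defaults to the standard Schlick value.
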MaterialFunction /Emporium/MaterialFunctions/MF_Fresnel_DS
graph {nodes [
  n1 = Multiply(A=FunctionInput(Preview=Constant() {r=0.5}) {bUsePreviewValueAsDefault=true, inputName=Specular, sortPriority=1}) {constB=0.08}
  n2 = OneMinus(None=@n1)
  n3 = Constant3Vector() {constant=<rgba 0 0 1 0>}
  n4 = DotProduct(A=Transform(None=FunctionInput(Preview=@n3) {bUsePreviewValueAsDefault=true, inputName=Normals TS}), B=CameraVectorWS())
  n5 = Power(
   Base=OneMinus(None=Abs(None=@n4) {desc="necessary for backfacing\r\nin RT mode (exit interface)"})
   Exp=FunctionInput(Preview=Constant() {desc="Exposed for artistic control.\r\nThe default of 5 is the common value for\r\nPBR schlick fresnel approximation.", r=5}) {bUsePreviewValueAsDefault=true, inputName=FresnelExponent, inputType=FunctionInput_Scalar, sortPriority=2}
  ) {constExponent=5}
  FunctionOutput(None=Add(A=@n1, B=Multiply(A=@n2, B=@n5)) {desc=Schlick Fresnel}) {outputName=Fresnel}
 ]}
