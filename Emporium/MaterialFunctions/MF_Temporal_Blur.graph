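# Material function that samples scene color behind a surface with a blur and normal-based refraction, outputting the color and distorted UVs.
# Blur uses one rotated offset per pixel, picked by a temporal-index dither plus screen-aligned noise and scaled by intensity over camera distance.
# A depth test drops blur offsets landing on foreground geometry.
# Refraction comes from camera-space normals, faded by camera distance relative to object size.
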MaterialFunction /Emporium/MaterialFunctions/MF_Temporal_Blur
graph {nodes [
  n1 = Constant3Vector() {constant=<rgba 0 0 1 0>}
  n2 = Transform(None=Multiply(A=FunctionInput(Preview=@n1) {bUsePreviewValueAsDefault=true, inputName=DistortionNormals, sortPriority=1}) {constB=-1}) {transformType=TRANSFORM_Camera}
  n3 = Constant()
  n4 = ScalarParameter() {defaultValue=0.5, group=Refraction, parameterName=RefractionAmount, sliderMax=1}
  n5 = Multiply(
   A=MaterialFunctionCall() {materialFunction=&ObjectLocalBounds}.Local Bounds Size
   B=MaterialFunctionCall() {materialFunction=&ObjectScale}.Scale XYZ
  )
  n6 = Divide(
   A=Distance(A=CameraPositionWS(), B=ObjectPositionWS())
   B=Multiply(A=MaterialFunctionCall(f3=@n5) {materialFunction=&AddComponents}.f3) {constB=2}
  ) {constB=1000, desc=ObjectSize}
  n7 = Multiply(
   A=Multiply(
    A=FunctionInput(Preview=@n3) {bUsePreviewValueAsDefault=true, inputName=RefractionAmount, inputType=FunctionInput_Scalar, sortPriority=3}
    B=Saturate(None=@n4)
   )
   B=OneMinus(None=Saturate(None=@n6))
  ) {desc=Distance Fade}
  n8 = PixelDepth()
  n9 = ScreenPosition()
  n10 = ComponentMask(None=FunctionInput(Preview=Constant()) {bUsePreviewValueAsDefault=true, inputName=BlurIntensity, inputType=FunctionInput_Scalar}) {r=true}
  n11 = ObjectPositionWS()
  n12 = CameraPositionWS()
  n13 = Divide(A=Multiply(A=@n10) {constB=10}, B=Multiply(A=Distance(A=@n11, B=@n12), B=Constant() {desc=was 0.2, r=0.2}))
  n14 = AppendVector(A=@n13, B=@n13)
  n15 = ViewProperty() {property=MEVP_TemporalSampleIndex}
  MaterialExpressionAdd_1 = Add(A=ScreenPosition().PixelPosition, B=AppendVector(A=@n15.Property, B=@n15.Property))
  n16 = Custom!(p=@MaterialExpressionAdd_1) {
   code="Mod( ((uint)(p.x) + 2 * (uint)(p.y)) , 5 )"
   outputType=CMOT_Float1
   inputs [{inputName=p, input {expression=&self:MaterialExpressionAdd_1}}]
  }
  n17 = TextureSample(UVs=MaterialFunctionCall!(TextureResolution=Constant2Vector() {g=64, r=64}) {materialFunction=/Engine/Functions/Engine_MaterialFunctions02/Texturing/ScreenAlignedPixelToPixelUVs}.Result) {
   samplerSource=SSM_Wrap_WorldGroupSettings
   samplerType=SAMPLERTYPE_LinearGrayscale
   texture=/Engine/EngineMaterials/Good64x64TilingNoiseHighFreq
  }
  n18 = Add(A=@n16, B=@n17.R) {constB=0}
  n19 = Constant() {r=5}
  n20 = Subtract(A=MaterialFunctionCall(UVs=Constant2Vector() {r=1}, "Rotation Angle (0-1)"=Divide(A=@n18, B=@n19) {constB=6}) {materialFunction=&CustomRotator}.Rotated Values) {constB=0.5}
  n21 = Multiply(A=Ceil(None=@n10), B=Multiply(A=Multiply(A=@n14, B=@n20), B=Constant2Vector() {g=2, r=1})) {desc=discard Offset if roughness is 0}
  n22 = SceneDepth(UVs=Add(A=ComponentMask(None=@n9.ViewportUV) {g=true, r=true}, B=@n21))
  n23 = Constant2Vector()
  n24 = Add(
   A=ComponentMask(None=Multiply(A=@n2, B=@n7) {constB=3}) {g=true, r=true}
   B=Add(A=If(A=@n8.R, B=@n22.R, "A > B"=@n23, "A == B"=@n23, "A < B"=@n21), B=ScreenPosition().ViewportUV)
  )
  FunctionOutput(None=SceneColor!(UVs=@n24).RGB) {outputName=SceneColor}
  FunctionOutput(None=@n24) {outputName=Distorted UVs, sortPriority=1}
 ]}
