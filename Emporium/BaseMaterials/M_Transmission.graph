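# Translucent Thin Translucent base material for transmissive surfaces, with index-of-refraction refraction and screen space reflections.
# A static switch selects specular-glossiness or metallic-roughness inputs, and transmission fades base color and metallic toward zero.
# MF_Temporal_Blur supplies distorted scene color, which is tinted and scaled by inverse Fresnel, with a constant replacing it on mobile.
# Opacity combines transmission with alpha coverage, and a thin-translucent output sets transmittance color.
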Material /Emporium/BaseMaterials/M_Transmission
graph {
 properties {
  bScreenSpaceReflections=true
  bUsedWithInstancedStaticMeshes=true
  blendMode=BLEND_Translucent
  opacityMaskClipValue=0.5
  pixelDepthOffsetMode=PDOM_Legacy
  refractionMethod=RM_IndexOfRefraction
  shadingModel=MSM_ThinTranslucent
  translucencyLightingMode=TLM_SurfacePerPixelLighting
  usedShadingModels=Thin Translucent
 }
 nodes [
  Constant() {r=1}
  n1 = MaterialFunctionCall() {materialFunction=/Emporium/MaterialFunctions/MF_Transmission}
  n2 = Min(A=OneMinus(None=NamedRerouteUsage()), B=@n1.Transmission) {desc=Roughness Contribution}
  LinearInterpolate(A=NamedRerouteUsage(), Alpha=Min(A=@n2, B=NamedRerouteUsage())) {constB=0, desc=No metallic when we fully transmit}
  LinearInterpolate(A=NamedRerouteUsage(), Alpha=Min(A=@n2, B=NamedRerouteUsage())) {constB=0, desc=Black when we fully transmit}
  n3 = MaterialFunctionCall() {materialFunction=/Emporium/MaterialFunctions/MF_DiffuseSpecGloss}
  n4 = MaterialFunctionCall!(DiffuseColor=@n3.DiffuseColor, SpecularColor=@n3.DiffuseColor, Glosiness=@n3.DiffuseColor) {materialFunction=/Emporium/MaterialFunctions/MF_SpecGlossToMetalRoughness}
  n5 = MaterialFunctionCall() {materialFunction=/Emporium/MaterialFunctions/MF_Basecolor}
  NamedRerouteDeclaration(None=StaticSwitchParameter(True=@n4."BaseColor (V3)", False=@n5.BaseColor) {group=PBRWorkflow, parameterName=Using SpecularGlossiness, sortPriority=0}) {name=BC, nodeColor=<rgba 0.8516 1 0 1>}
  n6 = MaterialFunctionCall() {materialFunction=/Emporium/MaterialFunctions/MF_MetallicRoughness}
  NamedRerouteDeclaration(None=StaticSwitchParameter(True=@n4.Roughness, False=@n6.Roughness) {group=PBRWorkflow, parameterName=Using SpecularGlossiness, sortPriority=0}) {name=Roughness, nodeColor=<rgba 0 1 0.4609 1>}
  NamedRerouteDeclaration(None=StaticSwitchParameter(True=@n4."Metallic (S)", False=@n6.Metallic) {group=PBRWorkflow, parameterName=Using SpecularGlossiness, sortPriority=0}) {name=Metallic, nodeColor=<rgba 1 0.7734 0 1>}
  NamedRerouteDeclaration(None=StaticSwitchParameter(True=@n3.Opacity, False=@n5.Opacity) {group=PBRWorkflow, parameterName=Using SpecularGlossiness, sortPriority=0}) {name=AlphaCoverage, nodeColor=<rgba 0.8516 1 0 1>}
  NamedRerouteUsage()
  n7 = MaterialFunctionCall() {materialFunction=/Emporium/MaterialFunctions/MF_Normals}
  Reroute(None=@n7.Result)
  n8 = MaterialFunctionCall() {materialFunction=/Emporium/MaterialFunctions/MF_IOR}
  n9 = MaterialFunctionCall() {materialFunction=/Emporium/MaterialFunctions/MF_Specular}
  n10 = Multiply(
   A=Multiply(A=Multiply(A=@n8.Specular, B=@n9.Specular), B=NamedRerouteUsage())
   B=MaterialFunctionCall() {materialFunction=/Emporium/MaterialFunctions/MF_Occlusion}.Occlusion
  )
  Reroute(None=@n10)
  n11 = OneMinus(None=@n1.Transmission)
  n12 = MaterialFunctionCall!(Normals TS=@n7.Result, Specular=@n10, FresnelExponent=Constant() {r=5}) {materialFunction=/Emporium/MaterialFunctions/MF_Fresnel_DS}
  n13 = MaterialFunctionCall!(BlurIntensity=NamedRerouteUsage(), DistortionNormals=@n7.Result, RefractionAmount=@n12.Fresnel) {materialFunction=/Emporium/MaterialFunctions/MF_Temporal_Blur}
  n14 = MaterialFunctionCall!(DistortedUVs=@n13.Distorted UVs, Roughness=NamedRerouteUsage(), Normals TS=@n7.Result) {materialFunction=/Emporium/MaterialFunctions/MF_TransmissionOpacity}
  Reroute(None=Min(A=NamedRerouteUsage(), B=Max(A=@n11, B=@n14.Opacity) {desc=Transmission}) {desc=alphaAsCoverage})
  n15 = Multiply(A=@n2, B=NamedRerouteUsage())
  n16 = OneMinus(None=@n12.Fresnel)
  ShadingPathSwitch(
   Default=Multiply(
    A=Multiply(A=@n15, B=@n13.SceneColor) {constA=1, desc=Tint}
    B=NamedRerouteDeclaration(None=@n16) {name=InverseFresnel, nodeColor=<rgba 0.4766 0 1 1>}
   ) {desc=Inverse Fresnel}
   Mobile=Constant3Vector()
  ) {desc=Migrated from FeatureLevelSwitch}
  ThinTranslucentMaterialOutput!(TransmittanceColor=LinearInterpolate(B=NamedRerouteUsage(), Alpha=NamedRerouteUsage()) {constA=1}, SurfaceCoverage)
 ]
}
